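; Sniper rifle weapon Blueprint for Athena that shows a scope glint while its wielder aims down sights.
; When targeting changes, it calls the parent handler, then passes the aim state to ShowSniperGlint only where cosmetic events can execute.
; ShowSniperGlint validates the instigator and Niagara glint component, then activates or deactivates the effect.
; The toggle runs only when the shooter is neither locally controlled nor on the local player's team.
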(event EventOnSetTargeting
  (params (bNewIsTargeting :type Boolean))
  ;[0 0 0] Event that is called on player started/stopped Aiming Down Sight
  (|CallParentFunction bNewIsTargeting)
  (Networking|CanExecuteCosmeticEvents
    (:True
      (Default|ShowSniperGlint bNewIsTargeting))))

(fn ShowSniperGlint
  (params (bAimedDownSight :type Boolean))
  (bind _returnvalue (Game|GetInstigator))
  (bind _ns_sniper_scope_glint (Variables|Default|GetNS_Sniper_Scope_Glint))
  ;[0 0 0] Activate or Deactivate the Sniper Glint Niagara System depending on if we have started or stopped Aiming Down Sight
  (Utilities|IsValid _returnvalue
    (:Is%20Valid
      (Variables|Default|GetNS_Sniper_Scope_Glint
        (:then
          (if (and (not (Pawn|IsLocallyControlled _returnvalue)) (not (Team|OnLocalPlayersTeam _returnvalue)))
            (if bAimedDownSight
              (Components|Activation|Activate _ns_sniper_scope_glint false)
              (else
                (Components|Activation|Deactivate _ns_sniper_scope_glint)))))
        (:else)))))

(fn ConstructionScript
  (|CallParentFunction))
