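; Heavy sniper rifle weapon Blueprint for Athena that shows a scope glint to opposing players while its user aims.
; On a targeting change it calls the parent, then toggles the Niagara glint only if the instigator is neither locally controlled nor on the local player's team.
; CalculateTraceDistance builds trace endpoints from player camera 0, reaching far along its forward vector.
; GetScopeParameters and the construction script defer to the parent.
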(event EventBeginPlay
  (|CallParentFunction) ; This node is disabled and will not be called.\nDrag off pins to build functionality.
)

(event EventTick
  (params (DeltaSeconds :type Float%20%28single-precision%29)))

(event EventOnSetTargeting
  (params (bNewIsTargeting :type Boolean))
  (bind _ns_sniper_scope_glint (Variables|Default|GetNS_Sniper_Scope_Glint))
  (|CallParentFunction bNewIsTargeting)
  (if (and (not (Pawn|IsLocallyControlled (Game|GetInstigator))) (not (Team|OnLocalPlayersTeam (Game|GetInstigator))))
    (Utilities|IsValid (Variables|Default|GetNS_Sniper_Scope_Glint)
      (:Is%20Valid
        (if bNewIsTargeting
          (Components|Activation|Activate _ns_sniper_scope_glint false)
          (else
            (Components|Activation|Deactivate _ns_sniper_scope_glint)))))))

(fn CalculateTraceDistance
  (return (+ (Math|Vector|vector+vector (Math|Vector|GetRightVector (Camera|GetCameraRotation (Game|GetPlayerCameraManager 0))) (Math|Vector|GetUpVector (Camera|GetCameraRotation (Game|GetPlayerCameraManager 0))) (Math|Vector|GetForwardVector (Camera|GetCameraRotation (Game|GetPlayerCameraManager 0)))) (Camera|GetCameraLocation (Game|GetPlayerCameraManager 0))) (+ (Camera|GetCameraLocation (Game|GetPlayerCameraManager 0)) (Math|Vector|vector+vector (Math|Vector|GetRightVector (Camera|GetCameraRotation (Game|GetPlayerCameraManager 0))) (Math|Vector|GetUpVector (Camera|GetCameraRotation (Game|GetPlayerCameraManager 0))) (* (Math|Vector|GetForwardVector (Camera|GetCameraRotation (Game|GetPlayerCameraManager 0))) 1e+05)))))

(fn GetScopeParameters
  (bind (_scopecomponent _depthoffieldvignetterange _weaponsightsoffset) (|CallParentFunction))
  (return _scopecomponent _depthoffieldvignetterange _weaponsightsoffset))

(fn ConstructionScript
  (|CallParentFunction))
