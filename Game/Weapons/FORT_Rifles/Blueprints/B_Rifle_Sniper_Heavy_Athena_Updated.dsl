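; Weapon Blueprint for an Athena heavy sniper rifle, adding a firing camera shake on top of its parent weapon behavior.
; The fire FX event calls the parent, then the shared WeaponFunctionLibrary FiringCameraShake.
; CalculateTraceDistance returns two points built from player camera 0: one offset slightly right, up and forward, the other projected far along the view direction.
; BeginPlay and Tick add no logic, and the construction script only calls its parent.
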(event EventBeginPlay
  (|CallParentFunction) ; This node is disabled and will not be called.\nDrag off pins to build functionality.
)

(event EventTick
  (params (DeltaSeconds :type Float%20%28single-precision%29)))

(event EventOnPlayWeaponFireFX
  (params (bPersistentFire :type Boolean) (bSecondaryFire :type Boolean))
  (|CallParentFunction bPersistentFire bSecondaryFire)
  (Class|WeaponFunctionLibrary|FiringCameraShake self))

(fn CalculateTraceDistance
  (return (+ (Math|Vector|vector+vector (Math|Vector|GetRightVector (Camera|GetCameraRotation (Game|GetPlayerCameraManager 0))) (Math|Vector|GetUpVector (Camera|GetCameraRotation (Game|GetPlayerCameraManager 0))) (Math|Vector|GetForwardVector (Camera|GetCameraRotation (Game|GetPlayerCameraManager 0)))) (Camera|GetCameraLocation (Game|GetPlayerCameraManager 0))) (+ (Camera|GetCameraLocation (Game|GetPlayerCameraManager 0)) (Math|Vector|vector+vector (Math|Vector|GetRightVector (Camera|GetCameraRotation (Game|GetPlayerCameraManager 0))) (Math|Vector|GetUpVector (Camera|GetCameraRotation (Game|GetPlayerCameraManager 0))) (* (Math|Vector|GetForwardVector (Camera|GetCameraRotation (Game|GetPlayerCameraManager 0))) 1e+05)))))

(fn ConstructionScript
  (|CallParentFunction))
