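; Suppressed thermal-scope sniper rifle weapon Blueprint that drives the thermal post-process view while aiming.
; Targeting changes on cosmetic clients with a local viewing controller update that controller's override of significance-based custom depth rendering.
; After a short delay, scope-in enables post-processing and animates the global thermal material parameter collection via a blend-time-scaled timeline, keeping players from seeing themselves.
; Forced scope-back also touches that timeline; reverse-scope only calls the parent and is noted as unused.
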(event EventOnSetTargeting
  (params (bNewIsTargeting :type Boolean))
  (Networking|CanExecuteCosmeticEvents
    (:True
      (bind _asfort_pawn (Utilities|Casting|CastToFortPawn (Game|GetInstigator)))
      (Utilities|IsValid (Pawn|GetLocalViewingPlayerController _asfort_pawn)
        (:Is%20Valid
          (|CallParentFunction bNewIsTargeting)
          (bind _asfort_pawn_2 (Utilities|Casting|CastToFortPawn (Game|GetInstigator)))
          (bind _asfort_player_controller (Utilities|Casting|CastToFortPlayerController (Pawn|GetLocalViewingPlayerController _asfort_pawn_2)))
          (Rendering|SetIgnoreSignifanceBasedCustomDepthRendering _asfort_player_controller (Camera|GetIsTargeting) 0))))))

(event EventPlayScopePP
  ;[1 1 1] Manually Enabling/Disabling PP so player does not see himself through Thermal Scope
  (Utilities|FlowControl|Sequence
    (:then_0
      (Utilities|FlowControl|Delay 0.1)
      (if (Camera|GetIsTargeting)
        (CallFunction|SetPostProcessParams 1.0)
        (Components|Timeline|SetPlayRate (Variables|B_Rifle_Sniper_Suppressed_ThermalScope_SpyTech_Athena|GetScalePlayerThermal) (/ 1.0 (Variables|Default|GetBlendInTime)))
        (Rendering|Material|SetScalarParameterValue "/Game/Athena/Items/Weapons/Abilities/Thermal/MPC_ThermalGlobal.MPC_ThermalGlobal" "ThermalFadeout" 1.0 :!class /Script/Engine.KismetMaterialLibrary)
        (Utilities|FlowControl|DoOnce false
          (:Completed
            (bind (_direction _newtrack_0) (Timeline|ScalePlayerThermal PlayFromStart)
              (:Update
                (Rendering|Material|SetScalarParameterValue "/Game/Athena/Items/Weapons/Abilities/Thermal/MPC_ThermalGlobal.MPC_ThermalGlobal" "ThermalScale" _newtrack_0 :!class /Script/Engine.KismetMaterialLibrary)))))))))

(event EventReverseScopePP
  ;[1 1 1] Manually Enabling/Disabling PP so player does not see himself through Thermal Scope
  ;;[0.1918 1 0.1475] No longer being called by parent
  (|CallParentFunction))

(event EventForceScopeBackImmediatly
  ;[1 1 1] Manually Enabling/Disabling PP so player does not see himself through Thermal Scope
  (|CallParentFunction)
  (Utilities|FlowControl|DoOnce false
    (:Completed
      (Timeline|ScalePlayerThermal PlayFromStart))))

(fn ConstructionScript
  (|CallParentFunction))
