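# Translucent unlit particle material for AMR rifle lightning arcs, usable on sprites, mesh particles and Niagara ribbons, with emissive from particle color.
# Muzzle-ring texture UVs get tiling, a curved Bend offset and optional axis swizzle.
# Floored, sine-modulated particle time steps the panner so arcs jump rather than scroll.
# Dynamic Erode adjusts the arc mask, minus a static texture mask and edge falloff, times particle alpha and optional vertex-color masking.
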Material /Game/Weapons/FORT_Rifles/Effects/Materials/M_AMR_LightningArcs_01
graph {
 properties {
  bEnableResponsiveAA=true
  bUsedWithMeshParticles=true
  bUsedWithNiagaraRibbons=true
  bUsedWithNiagaraSprites=true
  bUsedWithParticleSprites=true
  blendMode=BLEND_Translucent
  pixelDepthOffsetMode=PDOM_Legacy
  shadingModel=MSM_Unlit
  usedShadingModels=Unlit
 }
 nodes [
  n1 = ParticleColor()
  MaterialFunctionCall(Emissive=@n1.RGB) {materialFunction=&MF_FX_EmissiveAdjustments}
  n2 = ComponentMask(None=TextureCoordinate()) {g=true}
  n3 = OneMinus(None=@n2)
  n4 = Multiply(
   A=DynamicParameter() {defaultValue=<rgba 1 0 1 1>, paramNames [Panner Speed, Erode, V Tile, Bend]}.Bend
   B=Power(Base=Multiply(A=@n3, B=@n2)) {constExponent=3}
  ) {constA=25}
  n5 = Constant()
  n6 = Add(A=TextureCoordinate() {uTiling=2}, B=Constant2Vector() {r=-0.5})
  n7 = Add(A=TextureCoordinate() {vTiling=2}, B=Constant2Vector() {g=-0.5})
  n8 = ScalarParameter() {parameterName=Swizzle?}
  n9 = ScalarParameter() {defaultValue=1, parameterName=U Tile}
  n10 = Multiply(A=@n9, B=DynamicParameter() {parameterIndex=1, defaultValue=<rgba 1 0 1 1>, paramNames [UScale,Param2,Param3,Param4]}.UScale)
  n11 = ScalarParameter() {parameterName=Swizzle?}
  n12 = DynamicParameter() {defaultValue=<rgba 1 0 1 1>, paramNames [Panner Speed, Erode, V Tile, Bend]}
  n13 = ScalarParameter() {defaultValue=1, parameterName=V Tile}
  n14 = Multiply(
   A=Add(A=AppendVector(A=@n4, B=@n5), B=LinearInterpolate(A=@n6, B=@n7, Alpha=@n8))
   B=AppendVector(A=LinearInterpolate(A=@n9, B=@n10, Alpha=@n11), B=Multiply(A=@n12.V Tile, B=@n13) {constA=1})
  )
  n15 = MaterialFunctionCall(XY=@n14) {materialFunction=&Swizzle}
  n16 = ScalarParameter() {parameterName=Swizzle?}
  n17 = ParticleRelativeTime()
  n18 = StaticSwitchParameter(
   True=Multiply(A=DynamicParameter() {defaultValue=<rgba 1 0 1 1>, paramNames [Panner Speed, Erode, V Tile, Bend]}.Panner Speed, B=@n17)
   False=@n17
  ) {parameterName=Use PRT Multiply?}
  n19 = Sine(None=@n18) {period=0.77}
  n20 = Sine(None=@n18) {period=1.12}
  n21 = DynamicParameter() {defaultValue=<rgba 1 0 1 1>, paramNames [Panner Speed, Erode, V Tile, Bend]}
  n22 = Add(A=Abs(None=Add(A=@n19, B=@n20)), B=Add(A=Multiply(A=@n18, B=@n21.Panner Speed), B=@n21.Panner Speed))
  n23 = VectorParameter() {parameterName=PannerSpeed, defaultValue=<rgba 0 1.25 0 0>}
  n24 = TextureSample!(
   UVs=Panner(Coordinate=LinearInterpolate(A=@n14, B=@n15.YX, Alpha=@n16), Time=Floor(None=@n22), Speed=ComponentMask(None=@n23.RGB) {g=true, r=true}) {bFractionalPart=true, speedY=1.25}
   Tex
   Level
   Apply View MipBias
  ) {mipValueMode=TMVM_MipLevel, samplerType=SAMPLERTYPE_LinearColor, texture=/Game/Effects/Fort_Effects/Textures/Tile/T_AMR_MuzzleRing_01}
  n25 = Multiply(A=ParticleRelativeTime(), B=ScalarParameter() {defaultValue=2, parameterName=PRT Scale})
  n26 = Multiply(
   A=LinearInterpolate(A=@n24.B, B=@n24.B, Alpha=Clamp(None=@n25))
   B=Add(A=@n24.G, B=DynamicParameter() {defaultValue=<rgba 1 0 1 1>, paramNames [Panner Speed, Erode, V Tile, Bend]}.Erode)
  )
  n27 = TextureCoordinate()
  n28 = MaterialFunctionCall(XY=@n27) {materialFunction=&Swizzle}
  n29 = ScalarParameter() {parameterName=Swizzle?}
  n30 = TextureCoordinate()
  n31 = LinearInterpolate(A=ComponentMask(None=@n30) {g=true}, B=ComponentMask(None=@n30) {r=true}, Alpha=ScalarParameter() {parameterName=Swizzle?})
  n32 = Add(
   A=TextureSample!(UVs=LinearInterpolate(A=@n27, B=@n28.YX, Alpha=@n29), Tex, Apply View MipBias) {samplerType=SAMPLERTYPE_LinearColor, texture=&T_AMR_MuzzleRing_01}.A
   B=Power(Base=OneMinus(None=@n31)) {constExponent=4}
  )
  n33 = Multiply(A=@n1.A, B=Clamp(None=Subtract(A=@n26, B=@n32)))
  n34 = VertexColor()
  n35 = AppendVector(A=@n34, B=@n34)
  StaticSwitchParameter(
   True=Multiply(
    A=@n33
    B=StaticComponentMaskParameter(None=@n35) {defaultA=true, defaultB=true, defaultG=true, defaultR=true, parameterName=Vertex Color Component Mask}
   )
   False=@n33
  ) {parameterName=Use Vertex Color Mask?}
 ]
}
