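# Translucent, unlit, two-sided particle material for an AMR muzzle ring effect, usable on sprites and mesh particles.
# A tiled lightning texture pans on particle relative time, optionally in staccato steps built from two sines and a floor.
# Its green channel can blend two colors for emissive, replacing particle color; blue drives alpha with dynamic-parameter erosion, optional vertex-color erosion and camera depth fade.
# The rounded time-or-parameter LERP has no effect because both inputs are the blue channel.
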Material /Game/Weapons/FORT_Rifles/Effects/Materials/M_AMR_MuzzleRing_01
graph {
 properties {
  bEnableResponsiveAA=true
  bUsedWithMeshParticles=true
  bUsedWithNanite=true
  bUsedWithNiagaraMeshParticles=true
  bUsedWithNiagaraSprites=true
  bUsedWithParticleSprites=true
  blendMode=BLEND_Translucent
  pixelDepthOffsetMode=PDOM_Legacy
  shadingModel=MSM_Unlit
  twoSided=true
  usedShadingModels=Unlit
 }
 nodes [
  n1 = ScalarParameter() {defaultValue=1, parameterName=EmissiveMultiplier}
  n2 = TextureCoordinate()
  n3 = AppendVector(
   A=ScalarParameter() {defaultValue=1, parameterName=U Tile}
   B=Multiply(
    A=DynamicParameter() {defaultValue=<rgba 1 0 1 1>, paramNames [Panner Speed, Erode, V Tile, LERP]}.V Tile
    B=ScalarParameter() {defaultValue=1, parameterName=V Tile}
   )
  )
  n4 = Multiply(A=ParticleRelativeTime(), B=DynamicParameter() {defaultValue=<rgba 1 0 1 1>, paramNames [Panner Speed, Erode, V Tile, LERP]}.Panner Speed)
  n5 = Add(A=Sine(None=@n4) {period=0.77}, B=Sine(None=@n4) {period=1.12})
  n6 = Floor(None=Add(A=Abs(None=@n5), B=@n4))
  n7 = TextureSampleParameter2D(UVs=Panner(Coordinate=Multiply(A=@n2, B=@n3), Time=StaticSwitchParameter(True=@n6, False=@n4) {defaultValue=true, parameterName=UseStaccatoTiming?}) {bFractionalPart=true, speedY=1.21}) {parameterName=LightningTexture, samplerType=SAMPLERTYPE_LinearColor, texture=/Game/Effects/Fort_Effects/Textures/Tile/T_AMR_MuzzleRing_01}
  n8 = LinearInterpolate(
   A=VectorParameter() {parameterName=ColorLow, defaultValue=<rgba 1 1 1 1>}.RGB
   B=VectorParameter() {parameterName=ColorHigh, defaultValue=<rgba 1 1 1 1>}.RGB
   Alpha=@n7.G
  )
  n9 = ParticleColor()
  n10 = Clamp(None=@n9.RGB) {maxDefault=30}
  MaterialFunctionCall(Emissive=StaticSwitchParameter(True=Multiply(A=@n1, B=@n8), False=ShadingPathSwitch(Default=@n9.RGB, Mobile=@n10) {desc=Migrated from FeatureLevelSwitch}) {parameterName=UseColorRange}) {materialFunction=&MF_FX_EmissiveAdjustments}
  n11 = Clamp(None=Multiply(A=ParticleRelativeTime(), B=ScalarParameter() {defaultValue=1.35, parameterName=PRT Scale}))
  n12 = DynamicParameter() {defaultValue=<rgba 1 0 1 1>, paramNames [Panner Speed, Erode, V Tile, LERP]}
  n13 = LinearInterpolate(
   A=@n7.B
   B=@n7.B
   Alpha=MaterialFunctionCall!(In=StaticSwitchParameter(True=@n11, False=@n12.LERP) {defaultValue=true, parameterName=Use Particle Time LERP?}) {materialFunction=/Engine/Functions/Engine_MaterialFunctions02/Utility/Round}.Result
  ) {constAlpha=1}
  n14 = DynamicParameter() {defaultValue=<rgba 1 0 1 1>, paramNames [Panner Speed, Erode, V Tile, LERP]}
  n15 = VertexColor()
  n16 = StaticComponentMaskParameter(None=MaterialFunctionCall(R=@n15.R, G=@n15.R, B=@n15.R, A=@n15.R) {materialFunction=&AppendMany}.RGBA) {defaultA=true, parameterName=Select Vertex Color Channel to Erode}
  n17 = Constant()
  n18 = Multiply(A=@n13, B=Subtract(A=Add(A=@n13, B=@n14.Erode), B=StaticSwitchParameter(True=@n16, False=@n17) {parameterName=Use Vertex Color To Erode?}))
  n19 = ScalarParameter() {group=Fade, parameterName=CameraFade_Length, sortPriority=4}
  n20 = ScalarParameter() {group=Fade, parameterName=CameraFade_Offset, sortPriority=5}
  Multiply(
   A=Multiply(A=ParticleColor().A, B=Clamp(None=@n18))
   B=Power(
    Base=MaterialFunctionCall(Fade Length=@n19, Fade Offset=@n20) {materialFunction=&CameraDepthFade}.Result
    Exp=ScalarParameter() {defaultValue=1, group=Fade, parameterName=CameraFade_Power, sortPriority=6}
   )
  )
 ]
}
