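# Unlit, two-sided translucent material for a raygun laser beam, usable on beam trails, spline meshes and instanced meshes.
# It blends panning Phase 1 bits and core into Phase 2 tendrils by stepped particle blue, though DebugLERP defaults to a sine cycle.
# Particle red and green vary tendril tiling and optionally staccato panning; alpha slides a reveal texture.
# Noise and reveal form a hard erosion mask on opacity, with end fading by texture or object-radius gradient.
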Material /Game/Weapons/FORT_Rifles/Effects/Materials/M_Raygun_LaserBeam_01
graph {
 properties {
  bEnableResponsiveAA=true
  bUsedWithBeamTrails=true
  bUsedWithInstancedStaticMeshes=true
  bUsedWithSplineMeshes=true
  blendMode=BLEND_Translucent
  pixelDepthOffsetMode=PDOM_Legacy
  shadingModel=MSM_Unlit
  twoSided=true
  usedShadingModels=Unlit
 }
 nodes [
  n1 = TextureCoordinate()
  n2 = TextureCoordinate() {coordinateIndex=3}
  n3 = TextureSample(UVs=Panner(Coordinate=StaticSwitchParameter(True=@n1, False=@n2) {defaultValue=true, parameterName=Preview UVs}) {bFractionalPart=true, speedX=-8, speedY=4}) {texture=&T_NoisePack_01}
  n4 = StaticSwitchParameter(True=TextureCoordinate(), False=TextureCoordinate() {coordinateIndex=3}) {defaultValue=true, parameterName=Preview UVs}
  n5 = AppendVector(
   A=ScalarParameter() {defaultValue=0.8, group=Phase 1, parameterName=Bits A U Repeat}
   B=ScalarParameter() {defaultValue=1, group=Phase 1, parameterName=Bits A V Repeat}
  )
  n6 = Multiply(A=Time(), B=ScalarParameter() {defaultValue=1, group=Phase 1, parameterName=Pan Speed Master A})
  n7 = TextureSampleParameter2D(UVs=Panner(Coordinate=Multiply(A=@n4, B=@n5), Time=@n6) {bFractionalPart=true, speedX=-13}) {
   group=Phase 1
   parameterName=BeamTexturePhase1
   samplerType=SAMPLERTYPE_LinearColor
   texture=/Game/Effects/Fort_Effects/Textures/Tile/T_Raygun_Beam_01
  }
  n8 = StaticSwitchParameter(True=TextureCoordinate(), False=TextureCoordinate() {coordinateIndex=3}) {defaultValue=true, parameterName=Preview UVs}
  n9 = AppendVector(
   A=ScalarParameter() {defaultValue=1.5, group=Phase 1, parameterName=Bits B U Repeat}
   B=ScalarParameter() {defaultValue=1, group=Phase 1, parameterName=Bits B V Repeat}
  )
  n10 = TextureSampleParameter2D(UVs=Panner(Coordinate=Multiply(A=@n8, B=@n9), Time=@n6) {bFractionalPart=true, speedX=-19}) {
   group=Phase 1
   parameterName=BeamTexturePhase1
   samplerType=SAMPLERTYPE_LinearColor
   texture=&T_Raygun_Beam_01
  }
  n11 = Multiply(A=Ceil(None=Add(A=@n3.B) {constB=-0.4}), B=Add(A=@n7.B, B=@n10.B))
  n12 = TextureCoordinate()
  n13 = TextureCoordinate() {coordinateIndex=3}
  n14 = TextureSampleParameter2D(UVs=Panner(Coordinate=StaticSwitchParameter(True=@n12, False=@n13) {defaultValue=true, parameterName=Preview UVs}) {bFractionalPart=true, speedX=-4}) {
   group=Phase 1
   parameterName=BeamTexturePhase1
   samplerType=SAMPLERTYPE_LinearColor
   texture=&T_Raygun_Beam_01
  }
  n15 = Add(A=@n11, B=@n14.A) {constB=0}
  n16 = StaticSwitchParameter(True=TextureCoordinate(), False=TextureCoordinate() {coordinateIndex=3}) {defaultValue=true, parameterName=Preview UVs}
  n17 = AppendVector(
   A=ScalarParameter() {defaultValue=0.25, group=Phase 1, parameterName=Bits Core U Repeat}
   B=ScalarParameter() {defaultValue=1, group=Phase 1, parameterName=Bits Core V Repeat}
  )
  n18 = TextureSampleParameter2D(UVs=Panner(Coordinate=Multiply(A=@n16, B=@n17)) {bFractionalPart=true, speedX=-4}) {
   group=Phase 1
   parameterName=BeamTexturePhase1
   samplerType=SAMPLERTYPE_LinearColor
   texture=&T_Raygun_Beam_01
  }
  n19 = ScalarParameter() {defaultValue=0.4, group=Phase 2, parameterName=Tendrils A U Repeat}
  n20 = ParticleColor()
  n21 = Multiply(
   A=StaticSwitchParameter(True=TextureCoordinate(), False=TextureCoordinate() {coordinateIndex=3}) {defaultValue=true, parameterName=Preview UVs}
   B=AppendVector(A=Multiply(A=@n19, B=@n20.R), B=ScalarParameter() {defaultValue=1, group=Phase 2, parameterName=Tendrils A V Repeat})
  )
  n22 = AppendVector(A=@n20.R, B=Constant())
  n23 = Multiply(A=Time(), B=ScalarParameter() {defaultValue=12, group=Phase 2, parameterName=Pan Speed Master B})
  n24 = Multiply(A=@n20.R, B=@n23)
  n25 = StaticSwitchParameter(True=Ceil(None=@n24), False=@n24) {defaultValue=true, group=Phase 2, parameterName=StacattoPanning?}
  n26 = TextureSampleParameter2D(UVs=Panner(Coordinate=Add(A=@n21, B=@n22), Time=@n25) {bFractionalPart=true, speedX=-3.2}) {
   group=Phase 2
   parameterName=BeamTexturePhase2
   samplerType=SAMPLERTYPE_LinearColor
   texture=/Game/Effects/Fort_Effects/Textures/Tile/T_Raygun_Beam_02
  }
  n27 = ScalarParameter() {defaultValue=0.6, group=Phase 2, parameterName=Tendrils B U Repeat}
  n28 = ParticleColor()
  n29 = Multiply(
   A=StaticSwitchParameter(True=TextureCoordinate(), False=TextureCoordinate() {coordinateIndex=3}) {defaultValue=true, parameterName=Preview UVs}
   B=AppendVector(A=Multiply(A=@n27, B=@n28.G), B=ScalarParameter() {defaultValue=1, group=Phase 2, parameterName=Tendrils B V Repeat})
  )
  n30 = AppendVector(A=@n28.G, B=Constant())
  n31 = Multiply(A=@n28.G, B=@n23)
  n32 = StaticSwitchParameter(True=Ceil(None=@n31), False=@n31) {defaultValue=true, group=Phase 2, parameterName=StacattoPanning?}
  n33 = TextureSampleParameter2D(UVs=Panner(Coordinate=Add(A=@n29, B=@n30), Time=@n32) {bFractionalPart=true, speedX=-4.55}) {
   group=Phase 2
   parameterName=BeamTexturePhase2
   samplerType=SAMPLERTYPE_LinearColor
   texture=&T_Raygun_Beam_02
  }
  n34 = Add(
   A=DotProduct(A=@n26.RGB, B=VectorParameter() {group=Phase 2, parameterName=BeamTex_ChannelPicker_A, defaultValue=<rgba 1 0 0 0>}.RGB)
   B=DotProduct(A=@n33.RGB, B=VectorParameter() {group=Phase 2, parameterName=BeamTex_ChannelPicker_B, defaultValue=<rgba 0 1 0 0>}.RGB)
  )
  n35 = Time()
  n36 = ScalarParameter() {defaultValue=3, group=LERP, parameterName=Steps}
  n37 = Ceil(None=Multiply(A=ParticleColor().B, B=@n36))
  n38 = StaticSwitchParameter(True=ConstantBiasScale(None=Sine(None=@n35) {period=4}) {bias=0.95}, False=Saturate(None=Divide(A=@n37, B=@n36))) {group=LERP, parameterName=DebugLERP?}
  n39 = Power(
   Base=LinearInterpolate(A=Subtract(A=@n15, B=@n18.G), B=@n34, Alpha=@n38) {desc=Opacity Lerp}
   Exp=ScalarParameter() {defaultValue=4, group=Erode, parameterName=Lerp Exponent}
  ) {constExponent=10}
  n40 = ScalarParameter() {defaultValue=50, group=Erode, parameterName=Lerp Brightness}
  n41 = ScalarParameter() {defaultValue=1.6, group=Erode, parameterName=Erode U Repeat A}
  n42 = ScalarParameter() {defaultValue=0.4, group=Erode, parameterName=Erode V Repeat A}
  n43 = StaticSwitchParameter(True=TextureCoordinate(), False=TextureCoordinate() {coordinateIndex=3}) {defaultValue=true, parameterName=Preview UVs}
  n44 = Panner(Coordinate=Multiply(A=AppendVector(A=@n41, B=@n42), B=@n43), Time=@n25) {bFractionalPart=true, speedX=4.8, speedY=-0.4}
  n45 = ScalarParameter() {defaultValue=1.1, group=Erode, parameterName=Erode U Repeat B}
  n46 = ScalarParameter() {defaultValue=0.71, group=Erode, parameterName=Erode V Repeat B}
  n47 = Panner(Coordinate=Multiply(A=@n43, B=AppendVector(A=@n45, B=@n46)), Time=@n32) {bFractionalPart=true, speedX=2.3, speedY=0.3}
  n48 = ParticleColor()
  n49 = Add(
   A=Add(
    A=TextureSample(UVs=@n44) {texture=&T_NoisePack_01}.R
    B=TextureSample(UVs=@n47) {texture=&T_NoisePack_01}.G
   )
   B=TextureSample(UVs=Panner(Time=@n48.A) {speedX=5}) {samplerType=SAMPLERTYPE_LinearColor, texture=/Game/Effects/Fort_Effects/Textures/Tile/T_Raygun_BeamReveal_01}.B
  ) {constB=-0.4}
  n50 = Multiply(A=Saturate(None=Multiply(A=@n39, B=@n40)), B=Floor(None=Saturate(None=@n49)))
  n51 = TextureCoordinate() {uTiling=0.1}
  n52 = ObjectRadius()
  n53 = ComponentMask(None=TextureCoordinate()) {r=true}
  n54 = Multiply(A=OneMinus(None=@n53), B=@n53)
  n55 = Saturate(None=LinearInterpolate(B=Multiply(A=@n52) {constB=0.005}, Alpha=Multiply(A=@n54) {constB=4}))
  LinearInterpolate(
   A=Multiply(
    A=@n50
    B=TextureSample(UVs=@n51) {samplerType=SAMPLERTYPE_LinearColor, texture=&T_Raygun_BeamReveal_01}.G
   )
   B=Multiply(A=@n50, B=Power(Base=@n55))
   Alpha=ScalarParameter() {parameterName=UseObjectRadiusGradient}
  )
  n56 = VectorParameter() {group=Phase 1, parameterName=Color Outline Phase 1, defaultValue=<rgba 30 0.1059 0.0097 0>}
  n57 = VectorParameter() {group=Phase 1, parameterName=Color Core Phase 1, defaultValue=<rgba 10 1.25 3.1219 0>}
  n58 = Add(A=@n11, B=Saturate(None=ConstantBiasScale(None=@n14.R) {bias=-0.5, scale=2}))
  n59 = VectorParameter() {group=Phase 2, parameterName=Color Outline Phase 2, defaultValue=<rgba 30 0.1059 0.0097 0>}
  n60 = VectorParameter() {group=Phase 2, parameterName=Color Core Phase 2, defaultValue=<rgba 10 1.25 3.1219 0>}
  n61 = ConstantBiasScale(None=Saturate(None=@n34)) {bias=-0.5, scale=2}
  MaterialFunctionCall(Emissive=LinearInterpolate(A=LinearInterpolate(A=@n56.RGB, B=@n57.RGB, Alpha=@n58), B=LinearInterpolate(A=@n59.RGB, B=@n60.RGB, Alpha=@n61), Alpha=@n38) {desc=Emissive Lerp}) {materialFunction=&MF_FX_EmissiveAdjustments}
 ]
}
