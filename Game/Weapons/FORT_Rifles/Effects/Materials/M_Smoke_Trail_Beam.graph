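# Translucent, two-sided lit material for smoke trail beam particles in rifle effects, with color parameters for smoke and emissive glow.
# Opacity math combines particle color, a UV1 edge mask, view-angle fade and near-camera fade.
# A panned packed smoke texture has switchable channels, but it drives a lerp whose inputs are both particle red.
# Particle blue drives emissive color, and relative time scales an upward vector for trail rise.
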Material /Game/Weapons/FORT_Rifles/Effects/Materials/M_Smoke_Trail_Beam
graph {
 properties {
  bUsedWithBeamTrails=true
  blendMode=BLEND_Translucent
  pixelDepthOffsetMode=PDOM_Legacy
  translucencyLightingMode=TLM_VolumetricPerVertexNonDirectional
  twoSided=true
  usedShadingModels=Default Lit
 }
 nodes [
  n1 = ParticleColor()
  MaterialFunctionCall(Emissive=LinearInterpolate(B=VectorParameter() {parameterName=EmissiveColor, defaultValue=<rgba 1 0.4614 0.0871 0>}.RGB, Alpha=@n1.B)) {materialFunction=&MF_FX_EmissiveAdjustments}
  n2 = ComponentMask(None=TextureCoordinate() {coordinateIndex=1}) {r=true}
  n3 = Multiply(A=OneMinus(None=@n2), B=@n2)
  n4 = ParticleColor()
  n5 = ComponentMask(None=WorldPosition().XYZ) {r=true}
  n6 = ScalarParameter() {defaultValue=1024, parameterName=ScaleWorldPos}
  n7 = Multiply(A=TextureCoordinate() {vTiling=0}, B=AppendVector(A=ScalarParameter() {defaultValue=2, parameterName=SmokeTile}, B=Multiply(A=@n5, B=@n6)))
  n8 = TextureCoordinate() {coordinateIndex=1, uTiling=0}
  n9 = TextureSampleParameter2D(UVs=Panner(Coordinate=Add(A=@n7, B=@n8)) {speedX=-0.3}) {parameterName=TrailTexture_BChannel, texture=/Game/Effects/Fort_Effects/Textures/Smoke/T_Smoke_ThinTrail_Packed}
  n10 = StaticSwitchParameter(True=@n9.R, False=Constant()) {parameterName=UseRChannel}
  n11 = StaticSwitchParameter(True=@n9.G, False=Constant()) {parameterName=UseGChannel}
  n12 = Multiply(
   A=Clamp(None=LinearInterpolate(Alpha=@n3) {constB=4})
   B=LinearInterpolate(A=@n4.R, B=@n4.R, Alpha=If(A=@n10, B=@n11, "A > B"=@n10, "A == B"=@n9.B, "A < B"=@n11))
  )
  n13 = ScalarParameter() {defaultValue=12, parameterName=EdgeFade}
  Multiply(
   A=Multiply(
    A=Multiply(A=@n1.A, B=@n12)
    B=MaterialFunctionCall("FadeSensitivity (optional)"=@n13) {materialFunction=&FadeBasedOnViewAngle}.Result
   )
   B=MaterialFunctionCall(Distance=ScalarParameter() {defaultValue=400, parameterName=NearCameraFadeDistance}) {materialFunction=&Near_Camera_Fade}.Result
  )
  Multiply(
   A=Multiply(A=ParticleRelativeTime(), B=ScalarParameter() {defaultValue=5000, parameterName=WPO_SmokeTrailRise})
   B=Constant3Vector() {constant=<rgba 0 0 1 0>}
  )
  Multiply(A=Time())
  ParticleColor()
  VectorParameter() {parameterName=SmokeColor, defaultValue=<rgba 1 1 1 0>}
 ]
}
