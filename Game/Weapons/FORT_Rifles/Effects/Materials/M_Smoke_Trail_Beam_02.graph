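# Translucent, lit, two-sided beam-trail material for a weapon smoke trail, with SmokeColor and EmissiveColor parameters and a time-driven upward rise offset.
# A panning GatlingBeamDistort texture warps the trail texture's UVs and erodes its opacity, which also fades at both ends of UV1.
# Particle color channels drive distortion (R), erosion (G), emissive (B) and opacity (A).
# Near-camera fade applies except on low quality.
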Material /Game/Weapons/FORT_Rifles/Effects/Materials/M_Smoke_Trail_Beam_02
graph {
 properties {
  bUsedWithBeamTrails=true
  blendMode=BLEND_Translucent
  pixelDepthOffsetMode=PDOM_Legacy
  translucencyLightingMode=TLM_VolumetricPerVertexNonDirectional
  twoSided=true
  usedShadingModels=Default Lit
 }
 nodes [
  n1 = ParticleColor()
  MaterialFunctionCall(Emissive=LinearInterpolate(B=VectorParameter() {parameterName=EmissiveColor, defaultValue=<rgba 1 0.4614 0.0871 0>}.RGB, Alpha=@n1.B)) {materialFunction=&MF_FX_EmissiveAdjustments}
  Multiply(
   A=Multiply(A=ParticleRelativeTime(), B=ScalarParameter() {defaultValue=5000, parameterName=WPO_SmokeTrailRise})
   B=Constant3Vector() {constant=<rgba 0 0 1 0>}
  )
  n2 = ComponentMask(None=TextureCoordinate() {coordinateIndex=1}) {r=true}
  n3 = OneMinus(None=@n2)
  n4 = Clamp(None=LinearInterpolate(Alpha=Multiply(A=@n3, B=@n2)) {constB=4})
  n5 = Constant() {r=0.5}
  n6 = TextureSample(UVs=Panner(Coordinate=TextureCoordinate() {uTiling=2}) {bFractionalPart=true, speedX=1}) {texture=/Game/Weapons/FORT_Rifles/Effects/Textures/T_GatlingBeamDistort}
  n7 = Subtract(A=Multiply(A=MaterialFunctionCall(X=@n5, Y=@n6.G) {materialFunction=&MakeFloat2}.Result) {constB=2})
  n8 = Multiply(A=ParticleColor().R, B=ScalarParameter() {defaultValue=0.2, parameterName=DistortStrength}) {constB=0.2}
  n9 = TextureSampleParameter2D(UVs=Add(A=Panner() {bFractionalPart=true}, B=Multiply(A=@n7, B=@n8))) {parameterName=Trail Texture, texture=/Game/Weapons/FORT_Rifles/Effects/Textures/T_GatlingBeam}
  n10 = ParticleColor()
  n11 = ComponentMask(None=OneMinus(None=Abs(None=@n7))) {g=true}
  n12 = Subtract(A=Multiply(A=Add(A=@n10.G, B=@n11)) {constB=3}) {constB=2}
  n13 = Multiply(A=@n1.A, B=Multiply(A=Multiply(A=@n4, B=@n9.RGB), B=Clamp(None=@n12)))
  n14 = ScalarParameter() {defaultValue=400, parameterName=NearCameraFadeDistance}
  QualitySwitch(
   Default=Multiply(A=@n13, B=MaterialFunctionCall(Distance=@n14) {materialFunction=&Near_Camera_Fade}.Result)
   Low=@n13
  )
  VectorParameter() {parameterName=SmokeColor, defaultValue=<rgba 1 1 1 0>}
 ]
}
