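# Unlit, translucent, two-sided beam-trail material, named as the smoke trail for an Athena beam cannon.
# Particle color passes through HDR-remap and emissive-adjustment functions.
# Particle alpha drives a power exponent on second-UV U, shifting a gradient along the trail; this masks a laser-beam texture, is sharpened, then scaled by alpha.
# A quality switch adds near-camera fade to that mask, skipping it on Low.
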Material /Game/Weapons/FORT_Rifles/Effects/Materials/M_Smoke_Trail_Beam_Cannon_Athena
graph {
 properties {
  bUsedWithBeamTrails=true
  blendMode=BLEND_Translucent
  pixelDepthOffsetMode=PDOM_Legacy
  shadingModel=MSM_Unlit
  translucencyLightingMode=TLM_VolumetricPerVertexNonDirectional
  twoSided=true
  usedShadingModels=Unlit
 }
 nodes [
  n1 = ParticleColor()
  MaterialFunctionCall(Emissive=MaterialFunctionCall(LDR Input=Clamp(None=@n1.RGB), HDR Tint=@n1.RGB) {materialFunction=&DeriveHDRfromLDR}.RemappedRange) {materialFunction=&MF_FX_EmissiveAdjustments}
  n2 = ComponentMask(None=TextureCoordinate() {coordinateIndex=1}) {r=true}
  n3 = Power(Base=@n2, Exp=LinearInterpolate(Alpha=ParticleColor().A) {constA=35, constAlpha=0, constB=0.2}) {constExponent=1.5}
  Multiply(A=@n2, B=@n3)
  Power(Base=OneMinus(None=@n2)) {constExponent=0.6}
  n4 = Clamp(None=LinearInterpolate(Alpha=@n3) {constB=5})
  n5 = TextureSample() {texture=/Game/PvP/StrategicBuildingActors/Defensive/LaserFence/Assets/T_LaserBeamParticle}
  n6 = Multiply(A=@n1.A, B=Power(Base=Multiply(A=@n4, B=@n5.RGB)) {constExponent=3})
  n7 = ScalarParameter() {defaultValue=200, parameterName=NearCameraFadeDistance}
  QualitySwitch(
   Default=Multiply(A=@n6, B=MaterialFunctionCall(Distance=@n7) {materialFunction=&Near_Camera_Fade}.Result)
   Low=@n6
  )
 ]
}
