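# Translucent master material for weapon smoke trails on particle beams, with a hot emissive glow.
# Two panned samples of a soft smoke texture are blended and eroded, with tiling scaled by object radius.
# Particle color drives erosion (R), panning (G), glow strength and falloff (B), and opacity (A).
# A laser-beam texture shapes the two-color HDR glow; opacity gets end gradients and near-camera fade, which Low quality skips.
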Material /Game/Weapons/FORT_Rifles/Effects/Materials/M_Smoke_Trail_Beam_Master
graph {
 properties {
  bUsedWithBeamTrails=true
  blendMode=BLEND_Translucent
  pixelDepthOffsetMode=PDOM_Legacy
  translucencyLightingMode=TLM_VolumetricPerVertexNonDirectional
  usedShadingModels=Default Lit
 }
 nodes [
  Constant() {r=1}
  MaterialFunctionCall() {materialFunction=&DebugTimeSine}
  n1 = VectorParameter() {parameterName=SmokeColor, defaultValue=<rgba 0.3698 0.3698 0.3698 0>}
  n2 = TextureCoordinate() {vTiling=0.25}
  n3 = Constant2Vector() {g=0.33}
  n4 = TextureCoordinate() {vTiling=2.5}
  n5 = Constant2Vector() {g=-0.75}
  n6 = ParticleColor()
  n7 = ObjectRadius()
  n8 = ScalarParameter() {defaultValue=250, parameterName=SmokeTileDistance}
  n9 = Multiply(A=LinearInterpolate(A=Add(A=@n2, B=@n3), B=Add(A=@n4, B=@n5), Alpha=@n6.R), B=AppendVector(A=Divide(A=@n7, B=@n8), B=Constant() {r=1}))
  n10 = Add(A=@n6.G, B=Multiply(A=Time()) {constB=0.2})
  n11 = VectorParameter() {parameterName=SmokeSpeed, defaultValue=<rgba -0.8 0 0 1>}
  n12 = TextureSampleParameter2D(UVs=Panner(Coordinate=@n9, Time=@n10, Speed=@n11.RGB) {speedX=0.8}) {parameterName=TrailTexture_BChannel, samplerType=SAMPLERTYPE_Grayscale, texture=&T_Smoke_Trail_Soft}
  n13 = Multiply(A=@n11.RGB) {constB=0.6}
  n14 = TextureSampleParameter2D(UVs=Multiply(A=Panner(Coordinate=@n9, Time=@n10, Speed=@n13) {speedX=0.6}, B=Constant2Vector() {g=1.08, r=2.4}) {constB=2}) {
   parameterName=TrailTexture_BChannel_1
   samplerType=SAMPLERTYPE_Grayscale
   texture=&T_Smoke_Trail_Soft
  }
  n15 = LinearInterpolate(A=@n12.G, B=@n14.G) {constAlpha=0.35}
  n16 = Add(A=@n6.R, B=@n15)
  n17 = Clamp(None=Multiply(A=LinearInterpolate(Alpha=@n16) {constA=-3.5, constB=1.25}, B=@n15))
  n18 = Multiply(A=@n1.RGB, B=@n17)
  n19 = ComponentMask(None=TextureCoordinate()) {r=true}
  n20 = Divide(A=ObjectRadius(), B=ScalarParameter() {defaultValue=10000, parameterName=FadeDistance_Emissive}) {constB=10000}
  n21 = Clamp(None=Power(Base=@n6.B, Exp=ScalarParameter() {defaultValue=4, parameterName=EmissiveLengthPower}) {constExponent=4}) {desc=Tone down emissive gradient falloff at full brightness}
  n22 = LinearInterpolate(Alpha=Clamp(None=Subtract(A=@n20, B=@n21)) {minDefault=-0.05}) {constA=0.3, constB=60}
  n23 = ObjectRadius()
  n24 = OneMinus(None=@n19)
  n25 = Clamp(None=LinearInterpolate(B=Multiply(A=@n23) {constB=0.1}, Alpha=Multiply(A=@n24, B=@n19)) {constB=40})
  n26 = Multiply(
   A=Clamp(None=Power(Base=@n19, Exp=@n22))
   B=Multiply(A=TextureSample() {texture=/Game/PvP/StrategicBuildingActors/Defensive/LaserFence/Assets/T_LaserBeamParticle}.G, B=@n25)
  )
  n27 = Multiply(
   A=LinearInterpolate(
    A=VectorParameter() {parameterName=EmissiveColor, defaultValue=<rgba 10 0.3015 0 0>}.RGB
    B=VectorParameter() {parameterName=EmissiveColor_1, defaultValue=<rgba 5 2.2575 0.0387 0>}.RGB
    Alpha=@n26
   )
   B=@n26
  )
  n28 = MaterialFunctionCall(LDR Input=Clamp(None=@n27), HDR Tint=@n27) {materialFunction=&DeriveHDRfromLDR}
  n29 = ParticleColor() {desc="R = erosion, G = panning, B = emissive"}
  n30 = Clamp(None=@n29.B) {maxDefault=100}
  MaterialFunctionCall(Emissive=Add(A=Multiply(A=@n18) {constB=0.07}, B=Multiply(A=@n28.RemappedRange, B=@n30))) {materialFunction=&MF_FX_EmissiveAdjustments}
  Multiply(A=Add(A=@n1.RGB, B=@n18))
  n31 = Multiply(A=@n25, B=@n17)
  n32 = ObjectRadius()
  n33 = ComponentMask(None=TextureCoordinate()) {r=true}
  n34 = OneMinus(None=@n33)
  n35 = Saturate(None=LinearInterpolate(B=Multiply(A=@n32) {constB=0.1}, Alpha=Multiply(A=@n34, B=@n33)))
  n36 = Multiply(A=@n29.A, B=LinearInterpolate(A=@n31, B=Multiply(A=@n31, B=@n35), Alpha=ScalarParameter() {parameterName=UseObjectRadiusGradient}))
  n37 = Multiply(
   A=@n36
   B=MaterialFunctionCall(Distance=ScalarParameter() {defaultValue=400, parameterName=NearCameraFadeDistance}) {materialFunction=&Near_Camera_Fade}.Result
  )
  QualitySwitch(Default=@n37, Low=@n36, High=@n37, Medium=@n37)
 ]
}
