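# Translucent, two-sided lit master material for weapon smoke trails on Niagara ribbons and beam trails.
# Opacity combines UV edge masks, near-camera fade, optional inversion, particle alpha and depth fade.
# Two panning channel-masked textures feed a lerp whose endpoints are both particle red, so they cancel out.
# Smoke and emissive colors can switch to particle color, with emissive passing through a shared adjustment function.
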Material /Game/Weapons/FORT_Rifles/Effects/Materials/M_Smoke_Trail_Ribbon_Master
graph {
 properties {
  bUsedWithBeamTrails=true
  bUsedWithNiagaraRibbons=true
  blendMode=BLEND_Translucent
  pixelDepthOffsetMode=PDOM_Legacy
  twoSided=true
  usedShadingModels=Default Lit
 }
 nodes [
  n1 = ScalarParameter() {defaultValue=4, group=Alpha, parameterName=OpacityMaxRange, sortPriority=4}
  n2 = ComponentMask(None=TextureCoordinate()) {r=true}
  n3 = Multiply(A=OneMinus(None=@n2), B=@n2)
  n4 = Clamp(None=Multiply(A=LinearInterpolate(B=@n1, Alpha=@n3), B=ScalarParameter() {defaultValue=1, group=Alpha, parameterName=PushesMaskRight, sortPriority=3}))
  n5 = ComponentMask(None=TextureCoordinate()) {g=true}
  n6 = OneMinus(None=@n5)
  n7 = Clamp(None=LinearInterpolate(B=ScalarParameter() {defaultValue=4, group=Alpha, parameterName=OpacityMaxRange, sortPriority=4}, Alpha=Multiply(A=@n6, B=@n5)))
  n8 = OneMinus(None=ScalarParameter() {defaultValue=1, group=Alpha, parameterName=VertMaskInfluence, sliderMax=1, sortPriority=5})
  n9 = LinearInterpolate(B=Clamp(None=Add(A=@n7, B=@n8)), Alpha=ScalarParameter() {group=Alpha, parameterName=Mask TopBottom, sortPriority=2}) {constA=1}
  n10 = ParticleColor()
  n11 = TextureCoordinate() {coordinateIndex=1}
  n12 = VectorParameter() {group=Trail Textures, parameterName=TrailTexture_A_UVScale, sortPriority=3, defaultValue=<rgba 1 1 0 0>}
  n13 = ComponentMask(None=@n12.RGB) {g=true, r=true}
  n14 = VectorParameter() {group=Trail Textures, parameterName=TrailTexture_A_PannerSpeed, sortPriority=4, defaultValue=<rgba 0.1 0 0 0>}
  n15 = TextureSampleParameter2D(UVs=Panner(Coordinate=Multiply(A=@n11, B=@n13), Speed=ComponentMask(None=@n14.RGB) {g=true, r=true}) {speedX=0.1}) {group=Trail Textures, parameterName=TrailTexture_A, sortPriority=1, texture=/Game/Effects/Fort_Effects/Textures/Smoke/T_ShotGun_Trace_Packed}
  n16 = ChannelMaskParameter(None=@n15.RGBA) {group=Trail Textures, maskChannel=Green, parameterName=TrailTextureA_Mask, sortPriority=2, defaultValue=<rgba 0 1 0 0>}
  n17 = TextureCoordinate() {coordinateIndex=1}
  n18 = VectorParameter() {group=Trail Textures, parameterName=TrailTexture_B_UVScale, sortPriority=7, defaultValue=<rgba 2 1 0 0>}
  n19 = ComponentMask(None=@n18.RGB) {g=true, r=true}
  n20 = VectorParameter() {group=Trail Textures, parameterName=TrailTexture_B_PannerSpeed, sortPriority=8, defaultValue=<rgba 0.025 0 0 0>}
  n21 = TextureSampleParameter2D(UVs=Panner(Coordinate=Multiply(A=@n17, B=@n19), Speed=ComponentMask(None=@n20.RGB) {g=true, r=true}) {speedX=0.025}) {
   group=Trail Textures
   parameterName=TrailTexture_B
   sortPriority=5
   texture=/Game/Effects/Fort_Effects/Textures/Smoke/T_Smoke_ThinTrail_Packed
  }
  n22 = ChannelMaskParameter(None=@n21.RGBA) {group=Trail Textures, maskChannel=Blue, parameterName=TrailTextureB_Mask, sortPriority=6, defaultValue=<rgba 0 0 1 0>}
  n23 = LinearInterpolate(A=@n10.R, B=@n10.R, Alpha=Add(A=@n16.R, B=@n22.R))
  n24 = Constant() {r=1}
  n25 = ScalarParameter() {group=Trail Textures, parameterName=UseTrailTextures, sliderMax=1}
  n26 = Multiply(
   A=Multiply(A=Multiply(A=@n4, B=@n9), B=LinearInterpolate(A=@n23, B=@n24, Alpha=@n25))
   B=MaterialFunctionCall(Distance=ScalarParameter() {defaultValue=60, group=Alpha, parameterName=NearCameraFadeDistance, sortPriority=1}) {materialFunction=&Near_Camera_Fade}.Result
  )
  n27 = OneMinus(None=@n26)
  n28 = ScalarParameter() {group=Alpha, parameterName=Invert Alpha Mask}
  DepthFade(
   Opacity=Multiply(A=ParticleColor().A, B=LinearInterpolate(A=@n26, B=@n27, Alpha=@n28))
   FadeDistance=ScalarParameter() {parameterName=FadeDistance}
  )
  n29 = ParticleColor()
  n30 = ScalarParameter() {group=Color, parameterName=Use Particle Color}
  LinearInterpolate(A=VectorParameter() {group=Color, parameterName=SmokeColor, sortPriority=1, defaultValue=<rgba 1 1 1 0>}.RGB, B=@n29.RGB, Alpha=@n30)
  n31 = VectorParameter() {group=Color, parameterName=EmissiveColor, sortPriority=2, defaultValue=<rgba 1 0.4614 0.0871 0>}
  MaterialFunctionCall(Emissive=LinearInterpolate(A=LinearInterpolate(B=@n31.RGB), B=@n29.RGB, Alpha=@n30)) {materialFunction=&MF_FX_EmissiveAdjustments}
 ]
}
