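# Unlit additive spark material for sprite particles in Niagara and Cascade, filed under rifle weapon effects.
# A per-particle random value blends three color parameters (red, white and blue by default) to tint the spark texture's red channel; the mobile shading path uses only the middle color.
# Particle color and alpha scale the result, with a brightness boost on low quality, before it feeds a shared emissive adjustment function.
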Material /Game/Weapons/FORT_Rifles/Effects/Materials/M_Spark_TriColorGPU_01
graph {
 properties {
  bEnableResponsiveAA=true
  bUsedWithNiagaraSprites=true
  bUsedWithParticleSprites=true
  blendMode=BLEND_Additive
  pixelDepthOffsetMode=PDOM_Legacy
  shadingModel=MSM_Unlit
  usedShadingModels=Unlit
 }
 nodes [
  n1 = VectorParameter() {parameterName=A, defaultValue=<rgba 1 0.1293 0.1146 0>}
  n2 = VectorParameter() {parameterName=B, defaultValue=<rgba 1 1 1 0>}
  n3 = VectorParameter() {parameterName=C, defaultValue=<rgba 0.0833 0.1029 1 0>}
  n4 = ParticleRandom()
  n5 = Multiply(
   A=TextureSample() {texture=/Game/Weapons/FORT_Rifles/Effects/Textures/spark}.R
   B=ShadingPathSwitch(
    Default=MaterialFunctionCall(1=@n1.RGB, 2=@n2.RGB, 3=@n3.RGB, A=@n4) {materialFunction=&Lerp_Multiple_Float3}.Lerp 3 Inputs
    Mobile=@n2.RGB
   ) {desc=Migrated from FeatureLevelSwitch}
  )
  n6 = ParticleColor()
  n7 = Multiply(A=@n5, B=@n6.RGB)
  n8 = Multiply(A=Multiply(A=Constant() {r=1.6}, B=@n5), B=@n6.RGB)
  MaterialFunctionCall(Emissive=Multiply(A=QualitySwitch(Default=@n7, Low=@n8), B=@n6.A)) {materialFunction=&MF_FX_EmissiveAdjustments}
 ]
}
