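# Animation Blueprint for the Pasta Ripper rifle weapon mesh that animates it in response to aim-down-sights targeting.
# A two-state machine switches between an empty Idle state and a Targeting state that plays ADS_Into once, driven by bIsTargeting.
# The event graph casts OwningActor to FortWeapon and copies GetIsTargeting into that flag.
# The state machine output passes through DefaultSlot, so montages can override it.
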AnimBlueprint /Game/Weapons/FORT_Rifles/Mesh/Pasta_Ripper_Rifle_Rifle_AnimBP
target_skeleton=&/Hightower_Tomato_Rifle_Rifle_Skeleton
graphs [
 AnimGraph {nodes [
   AnimGraphNode_StateMachine_0 = New State Machine() {is=title AnimGraphNode_StateMachine}
   Output Pose(Result=Slot 'DefaultSlot'(Source=@AnimGraphNode_StateMachine_0)) {layerGroup=None}
  ]}
 New State Machine {
  entry=Idle
  path=AnimGraph.AnimGraphNode_StateMachine_0.New State Machine
  ids=id
  nodes [
   AnimStateNode_0 = Idle()
   AnimStateNode_1 = Targeting() {is=title AnimStateNode}
   AnimStateTransitionNode_0 = AnimStateTransitionNode()
   AnimStateTransitionNode_1 = AnimStateTransitionNode()
  ]
  links [
   {
    dst=Targeting
    rule=AnimGraph.AnimGraphNode_StateMachine_0.New State Machine.AnimStateTransitionNode_0.Transition
    src=Idle
    via=AnimStateTransitionNode_0
   }
   {
    dst=Idle
    rule=AnimGraph.AnimGraphNode_StateMachine_0.New State Machine.AnimStateTransitionNode_1.Transition
    src=Targeting
    via=AnimStateTransitionNode_1
   }
  ]
 }
 Idle {path=AnimGraph.AnimGraphNode_StateMachine_0.New State Machine.AnimStateNode_0.Idle, nodes [Output Animation Pose()]}
 Targeting {
  path=AnimGraph.AnimGraphNode_StateMachine_0.New State Machine.AnimStateNode_1.Targeting
  nodes [Output Animation Pose(Result=Sequence Player 'ADS_Into'() {bLoopAnimation=false, sequence=/Game/Animation/Game/MainPlayer/Combat/Guns/Pasta_Ripper/ADS_Into})]
 }
 Transition {
  path=AnimGraph.AnimGraphNode_StateMachine_0.New State Machine.AnimStateTransitionNode_0.Transition
  nodes [Result(bCanEnterTransition=Get bIsTargeting())]
 }
 Transition {
  path=AnimGraph.AnimGraphNode_StateMachine_0.New State Machine.AnimStateTransitionNode_1.Transition
  nodes [Result(bCanEnterTransition=NOT Boolean(A=Get bIsTargeting()))]
 }
 EventGraph {nodes [
   n1 = Cast To FortWeapon!(execute, Object=Get OwningActor() -> (OwningActor:Actor Object Reference)) -> (then, CastFailed, AsFort Weapon:Fort Weapon Object Reference) {is=title K2Node_DynamicCast}
   > Set bIsTargeting(bIsTargeting:bool=GetIsTargeting!(self:Fort Weapon Object Reference=@n1) -> (ReturnValue:bool)) -> (Output_Get:bool)
   TryGetPawnOwner()
   Parent: BlueprintInitializeAnimation() {is=title K2Node_CallParentFunction}
   Event BlueprintUpdateAnimation()
   Event BlueprintInitializeAnimation()
   Set OwningActor(OwningActor:Actor Object Reference=GetOwningActor()) -> (Output_Get:Actor Object Reference)
  ]}
]
variables []
