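# Physics asset for the SK_Assault_Bullpup_01 skeletal mesh, a bullpup assault rifle, giving it a simplified collision shape.
# A single capsule body on the Handle bone, offset and rotated from that bone, stands in for the weapon.
# It has no constraints, so it defines one rigid body with no jointed parts.
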PhysicsAsset /Game/Weapons/FORT_Rifles/Mesh/SK_Assault_Bullpup_01_PhysicsAsset
bodies {Handle [{shapeType=Capsule, length=43.3168, radius=16.4421, center=<13.73 0.0604 5.3928>, rotation=<rot 0 90.0924 -84.7079>}]}
constraints []
