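# Physics asset for the suppressed AutoHigh rifle skeletal mesh, SK_AutoHigh_Suppressed, giving the weapon a simple physics shape.
# It holds a single body on the Handle bone, built from one long capsule that is offset and rotated from the bone.
# There are no constraints, so the physics representation is one piece with no jointed parts.
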PhysicsAsset /Game/Weapons/FORT_Rifles/Mesh/SK_AutoHigh_Suppressed_PhysicsAsset
bodies {Handle [{shapeType=Capsule, length=88.7332, radius=17.7098, center=<27.9926 0.5451 3.695>, rotation=<rot 0 89.9695 -85.5966>}]}
constraints []
