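# Physics asset for the BattleLab ElephantGarlic rifle skeletal mesh, giving the weapon a simple collision shape.
# It holds a single box body on the Handle bone, offset mostly forward and up from the bone origin to enclose the gun as one rigid shape.
# There are no constraints, so no part of the weapon articulates in simulation.
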PhysicsAsset /Game/Weapons/FORT_Rifles/Mesh/SK_BattleLab_ElephantGarlic_Physics
bodies {Handle [{shapeType=Box, extentX=72.7372, extentY=15.277, extentZ=34.6126, center=<12.019 -0.1483 6.7178>}]}
constraints []
