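# Physics asset for the SK_BattleRifleSixties rifle skeletal mesh, giving the weapon a simplified collision shape.
# It holds a single body on the Handle bone, made of one capsule offset and rotated from that bone, apparently to lie along the length of the rifle.
# It defines no constraints, so the rifle simulates as one rigid piece with no articulated parts.
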PhysicsAsset /Game/Weapons/FORT_Rifles/Mesh/SK_BattleRifleSixties_PhysicsAsset
bodies {Handle [{shapeType=Capsule, length=55.6185, radius=11.7814, center=<29.7113 0.1825 5.0144>, rotation=<rot 0 89.9141 -85.8325>}]}
constraints []
