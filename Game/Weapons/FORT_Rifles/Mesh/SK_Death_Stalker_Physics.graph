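# Physics asset for the Death Stalker rifle skeletal mesh, defining its simple collision shape for physics simulation.
# It has a single body on the Handle bone, one box offset from the bone origin that is long on X and thin on Y.
# No constraints exist, so the weapon behaves as one rigid piece.
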PhysicsAsset /Game/Weapons/FORT_Rifles/Mesh/SK_Death_Stalker_Physics
bodies {Handle [{shapeType=Box, extentX=99.0591, extentY=7.795, extentZ=47.0426, center=<23.2957 0.3822 6.5702>}]}
constraints []
