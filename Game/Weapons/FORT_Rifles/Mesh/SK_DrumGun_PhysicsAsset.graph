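# Physics asset for the Drum Gun rifle skeletal mesh, giving the weapon simple collision.
# A single body on the Handle bone uses one rotated, offset box, longest along its local Z axis.
# With no other bodies and no constraints, the gun acts as one rigid piece with no articulated parts.
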PhysicsAsset /Game/Weapons/FORT_Rifles/Mesh/SK_DrumGun_PhysicsAsset
bodies {Handle [{shapeType=Box, extentX=7.416, extentY=22.0816, extentZ=76.5374, center=<9.8587 0.5348 1.1115>, rotation=<rot 0.0006 90.0131 -89.9377>}]}
constraints []
