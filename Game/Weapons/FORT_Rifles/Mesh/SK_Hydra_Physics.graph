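# Physics asset for the Hydra rifle skeletal mesh, SK_Hydra, in the Fortnite rifle weapons folder.
# It holds a single box body on the Handle bone, sized and offset to approximate the weapon as one rigid shape.
# It has no constraints, so it defines no articulated or jointed parts.
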PhysicsAsset /Game/Weapons/FORT_Rifles/Mesh/SK_Hydra_Physics
bodies {Handle [{shapeType=Box, extentX=81.5685, extentY=14.7263, extentZ=41.0645, center=<13.7321 0.1285 3.6289>}]}
constraints []
