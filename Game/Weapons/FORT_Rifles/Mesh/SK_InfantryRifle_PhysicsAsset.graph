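# Physics asset for the infantry rifle skeletal mesh, giving the weapon a simplified collision and physics shape.
# It holds a single body on the Handle bone, a capsule rotated and offset along the rifle.
# It has no constraints, so the weapon behaves as one rigid piece with no articulated parts.
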PhysicsAsset /Game/Weapons/FORT_Rifles/Mesh/SK_InfantryRifle_PhysicsAsset
bodies {Handle [{shapeType=Capsule, length=57.9988, radius=11.4697, center=<15.3953 0.0001 2.2943>, rotation=<rot 90 0 0>}]}
constraints []
