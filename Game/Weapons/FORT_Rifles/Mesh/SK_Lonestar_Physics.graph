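# Physics asset for the Lonestar rifle skeletal mesh, providing its simple collision and physics shape.
# It holds a single body on the Handle bone, one offset box that is long along X and thin along Y.
# No constraints are defined, so the weapon behaves as one rigid piece in physics.
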PhysicsAsset /Game/Weapons/FORT_Rifles/Mesh/SK_Lonestar_Physics
bodies {Handle [{shapeType=Box, extentX=89.0185, extentY=8.8453, extentZ=27.8765, center=<15.5128 0.4037 1.0915>}]}
constraints []
