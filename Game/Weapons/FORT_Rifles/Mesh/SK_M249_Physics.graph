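# Physics asset for the M249 rifle skeletal mesh, providing its simplified collision shape.
# A single box body on the M249_Handle bone, offset and rotated to lie along the weapon, supplies all collision.
# It defines no constraints, so the weapon has one rigid body with no articulated parts.
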PhysicsAsset /Game/Weapons/FORT_Rifles/Mesh/SK_M249_Physics
bodies {M249_Handle [{shapeType=Box, extentX=23.6736, extentY=36.4633, extentZ=105.7049, center=<0.801 26.5907 3.2479>, rotation=<rot 0 1.7025 -89.6772>}]}
constraints []
