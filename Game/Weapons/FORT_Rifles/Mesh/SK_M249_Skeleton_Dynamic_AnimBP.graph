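# Animation Blueprint for the M249 rifle skeleton that adds physics-driven secondary motion to an input pose.
# The pose passes through a RigidBody node in component space, with no override physics asset shown, then goes to the output.
# Two Anim Dynamics nodes for the handle and bipod bones, with angular limits and low gravity, are left unconnected.
# The event graph holds only an empty update event.
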AnimBlueprint /Game/Weapons/FORT_Rifles/Mesh/SK_M249_Skeleton_Dynamic_AnimBP
target_skeleton=&/SK_M249_Skeleton
graphs [
 AnimGraph {nodes [
   Anim Dynamics - Bone: dyn_Handle(Alpha=1, ExternalForce="0,0,0") {
    angularBiasOverride=1
    bGravityOverrideInSimSpace=true
    bOverrideAngularBias=true
    gravityScale=0.1
    alphaBoolBlend {blendOption=Linear}
    boundBone {boneName=dyn_Handle}
    physicsBodyDefinitions [{
      sphereCollisionRadius=0
      boundBone {boneName=dyn_Handle}
      boxExtents=<5 5 5>
      constraintSetup {
       angularTargetAxis=AxisY
       twistAxis=AxisY
       angularLimitsMax=<0 15 0>
       angularLimitsMin=<0 -15 0>
       angularTarget=<0 0 0>
       linearAxesMax=<0 0 3>
      }
      localJointOffset=<0 0 -2>
     }]
   }
   Anim Dynamics - Bone: dyn_biPod(Alpha=1, ExternalForce="0,0,0") {
    angularBiasOverride=1
    bGravityOverrideInSimSpace=true
    bOverrideAngularBias=true
    gravityScale=0.1
    alphaBoolBlend {blendOption=Linear}
    boundBone {boneName=dyn_biPod}
    physicsBodyDefinitions [{
      sphereCollisionRadius=0
      boundBone {boneName=dyn_biPod}
      boxExtents=<5 5 5>
      constraintSetup {
       angularTargetAxis=AxisY
       twistAxis=AxisY
       angularLimitsMax=<15 0 0>
       angularLimitsMin=<-15 0 0>
       angularTarget=<0 0 0>
       linearAxesMax=<0 0 3>
      }
      localJointOffset=<0 0 -10>
     }]
   }
   AnimGraphNode_SubInput_0 = Input Pose()
   n1 = RigidBody(ComponentPose=Local To Component(LocalPose=@AnimGraphNode_SubInput_0)) {alphaBoolBlend {blendOption=Linear}, componentAppliedLinearAccClamp=<0 0 0>, componentLinearAccScale=<1 1 1>}
   Output Pose(Result=Component To Local(ComponentPose=@n1)) {layerGroup=None}
  ]}
 EventGraph {nodes [Event BlueprintUpdateAnimation()]}
]
variables []
