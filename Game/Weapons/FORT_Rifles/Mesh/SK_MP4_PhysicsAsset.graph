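# Physics asset for the MP4 weapon skeletal mesh in the FORT rifles set, giving it a simplified physics shape.
# A single box body on the Handle bone, elongated along X and offset forward from the bone, serves as the only collision shape.
# It has no constraints, so the asset defines one rigid body with no articulated parts.
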PhysicsAsset /Game/Weapons/FORT_Rifles/Mesh/SK_MP4_PhysicsAsset
bodies {Handle [{shapeType=Box, extentX=65.5, extentY=7.5, extentZ=36, center=<14.8312 0 -0.7029>}]}
constraints []
