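# Physics asset for the SK_MachineGun_01 skeletal mesh, a machine gun in the FORT rifles weapon set.
# A single box body on the Handle bone approximates the gun with one long, narrow volume, offset mostly along its length.
# It has no constraints, so the physics representation is one rigid piece with no articulated parts.
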PhysicsAsset /Game/Weapons/FORT_Rifles/Mesh/SK_MachineGun_01_Physics2
bodies {Handle [{shapeType=Box, extentX=91.8705, extentY=7.8815, extentZ=27.8711, center=<16.5654 0.1164 4.7643>}]}
constraints []
