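# Physics asset for the silenced military SMG skeletal mesh among the Fortnite rifle weapons.
# A single box body on the Handle bone approximates the whole gun, offset forward and slightly down from that bone.
# It has no constraints, so the weapon simulates and collides as one rigid piece with no articulated parts.
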PhysicsAsset /Game/Weapons/FORT_Rifles/Mesh/SK_MilitarySMG_Silenced_PhysicsAsset
bodies {Handle [{shapeType=Box, extentX=73, extentY=7.5, extentZ=33.5, center=<13.6268 0 -1.8347>}]}
constraints []
