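# Physics asset for the military M4A1 rifle skeletal mesh in the Fort rifle weapon set.
# It holds a single box body on the Handle bone, offset and rotated as one simple collision shape.
# With no constraints, the rifle simulates as one rigid piece rather than articulated parts.
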PhysicsAsset /Game/Weapons/FORT_Rifles/Mesh/SK_Military_M4A1_PhysicsAsset
bodies {Handle [{shapeType=Box, extentX=8.916, extentY=38.0816, extentZ=99.0374, center=<23.84 0.534 5.1115>, rotation=<rot 0 90.0133 -87.4406>}]}
constraints []
