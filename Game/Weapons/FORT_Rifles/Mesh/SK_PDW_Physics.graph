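# Physics asset for the SK_PDW personal defense weapon skeletal mesh in the Fortnite rifle set.
# It holds a single capsule body on the P90_Handle bone, offset and rotated from the bone origin to act as one rough volume for the weapon.
# It defines no constraints, so the weapon simulates and collides as one rigid piece rather than an articulated ragdoll.
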PhysicsAsset /Game/Weapons/FORT_Rifles/Mesh/SK_PDW_Physics
bodies {P90_Handle [{shapeType=Capsule, length=34.3025, radius=14.0474, center=<11.1431 0.0265 6.231>, rotation=<rot 0 89.649 -87.8658>}]}
constraints []
