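# Physics asset for the Post Apocalyptic assault rifle skeletal mesh, providing its simple collision shape.
# It holds a single body on the Handle bone, made of one box offset from the bone to cover the weapon.
# It defines no constraints, so the rifle is represented as one rigid piece with no jointed parts.
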PhysicsAsset /Game/Weapons/FORT_Rifles/Mesh/SK_Post_Apocalyptic_Assault_Physics
bodies {Handle [{shapeType=Box, extentX=85.3913, extentY=10.2951, extentZ=26.4269, center=<16.4318 1.7307 4.8538>}]}
constraints []
