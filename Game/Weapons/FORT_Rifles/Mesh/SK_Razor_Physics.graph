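# Physics asset for the SK_Razor rifle skeletal mesh, providing simple collision for the weapon.
# It holds a single body on the Handle bone, made of one box shape offset slightly from the bone to cover the gun.
# There are no constraints, so the weapon behaves as one rigid piece rather than an articulated set of bodies.
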PhysicsAsset /Game/Weapons/FORT_Rifles/Mesh/SK_Razor_Physics
bodies {Handle [{shapeType=Box, extentX=73.1132, extentY=9.7974, extentZ=29.9081, center=<7.4684 0.9684 3.1956>}]}
constraints []
