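# Physics asset for the SK_SAW_01 rifle skeletal mesh, giving the weapon a simple collision shape.
# It holds a single body on the Handle bone, a box offset from the bone origin and elongated along X.
# There are no constraints, so the weapon simulates as one rigid piece with no articulated parts.
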PhysicsAsset /Game/Weapons/FORT_Rifles/Mesh/SK_SAW_01_Physics2
bodies {Handle [{shapeType=Box, extentX=115.212, extentY=13.3384, extentZ=39.0194, center=<26.8306 1.6637 9.826>}]}
constraints []
