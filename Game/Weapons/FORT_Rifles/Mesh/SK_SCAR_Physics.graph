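# Physics asset for the SCAR rifle skeletal mesh, giving the weapon a simple collision and physics shape.
# A single box body on the Handle bone, offset from the bone origin, stands in for the whole rifle.
# It has no constraints, so the asset defines one rigid piece with no jointed parts.
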PhysicsAsset /Game/Weapons/FORT_Rifles/Mesh/SK_SCAR_Physics
bodies {Handle [{shapeType=Box, extentX=96.1416, extentY=9.7743, extentZ=33.5858, center=<20.9778 0.5488 2.3135>}]}
constraints []
