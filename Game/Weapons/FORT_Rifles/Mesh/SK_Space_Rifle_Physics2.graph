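# Physics asset for the SK_Space_Rifle skeletal mesh, a space-themed rifle among the FORT_Rifles weapons.
# A single box body on the Handle bone, offset from the bone origin, provides the rifle's whole collision shape.
# It defines no constraints, so the weapon simulates as one rigid piece with no articulated parts.
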PhysicsAsset /Game/Weapons/FORT_Rifles/Mesh/SK_Space_Rifle_Physics2
bodies {Handle [{shapeType=Box, extentX=62.2961, extentY=10.2254, extentZ=34.6398, center=<15.9302 0.6972 5.3795>}]}
constraints []
